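# Translucent, refractive glass material for the Distant Echo Pro glider, supporting skeletal meshes.
# Fresnel over a scaled noise normal drives opacity, with a switchable dirt mask addition; the same mask tints base color.
# Emissive fakes reflections with a desaturated, ambient-tinted HDR cubemap plus a specular highlight from the sun/moon collection vector.
# Noise and asset normals are summed, and all attributes pass through the accessories tech-art override function.
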Material /BRCosmetics/Gadgets/Assets/VinderTech_GliderChute/Glider_DistantEchoPro/Materials/M_Glider_DistantEchoPro_Glass
graph {
 properties {
  bEnableResponsiveAA=true
  bScreenSpaceReflections=true
  bUseMaterialAttributes=true
  bUsedWithMeshDeformer=true
  bUsedWithSkeletalMesh=true
  blendMode=BLEND_Translucent
  physMaterial=/Game/FortressPhysicalMaterials/Glass
  pixelDepthOffsetMode=PDOM_Legacy
  refractionMethod=RM_IndexOfRefraction
  translucencyDirectionalLightingIntensity=5
  translucencyLightingMode=TLM_VolumetricPerVertexNonDirectional
  usedShadingModels=Default Lit
 }
 nodes [
  n1 = TextureCoordinate() {vTiling=0.2}
  n2 = Normalize(VectorInput=Multiply(
    A=TextureSample(UVs=@n1) {samplerType=SAMPLERTYPE_Normal, texture=/Game/Global/Textures/Default/Noise/T_Noise_Normal_01}.RGB
    B=VectorParameter() {parameterName=Normal Intensity, defaultValue=<rgba 0.1 0.1 1 0>}.RGB
   ))
  n3 = TextureSampleParameter2D() {
   parameterName=N
   samplerType=SAMPLERTYPE_Normal
   texture=/Game/Gadgets/Assets/VinderTech_GliderChute/Glider_DistantEchoPro/Textures/T_Glider_DistantEchoPro_N
  }
  n4 = TextureSampleParameter2D() {
   group=Dirt
   parameterName=FX
   samplerType=SAMPLERTYPE_Masks
   texture=/Game/Gadgets/Assets/VinderTech_GliderChute/Glider_DistantEchoPro/Textures/T_Glider_DistantEchoPro_FX
  }
  n5 = Multiply(A=@n4.R, B=ScalarParameter() {group=Dirt, parameterName=Dirty_Power})
  n6 = LinearInterpolate(
   A=ScalarParameter() {parameterName=Fresnel Inner Transparency}
   B=ScalarParameter() {defaultValue=0.85, parameterName=Fresnel Outer Transparency}
   Alpha=Fresnel(ExponentIn=ScalarParameter() {defaultValue=4, parameterName=Fresnel Exponent}, Normal=Transform(None=@n2))
  )
  n7 = ReflectionVectorWS() {bNormalizeCustomWorldNormal=true}
  n8 = Multiply(
   A=Multiply(
    A=TextureSampleParameterCube!(UVs=@n7, Apply View MipBias) {parameterName=MaterialExpressionTextureSampleParameterCube_0, texture=/Game/Packages/Fortress_Sky/TexturesHDR/T_hdrsky_test03_low}.RGB
    B=VectorParameter() {parameterName=Cube Map Multiplier, defaultValue=<rgba 1.5 1.5 1.5 0>}.RGB
   ) {constB=1.5}
   B=ComponentMask(None=MaterialFunctionCall() {materialFunction=/Engine/Functions/GetPostProcessFunctions/GetAmbientCubemapTint}.AmbientCubemapTint) {b=true, g=true, r=true}
  )
  n9 = Constant() {r=0.5}
  n10 = CollectionParameter() {
   collection=&FortniteMaterialParameters
   parameterName=SunAndMoonModelDirectionalVector
  }
  n11 = ComponentMask(None=@n10) {b=true, g=true, r=true}
  n12 = ReflectionVectorWS()
  n13 = Power(Base=Clamp(None=DotProduct(A=@n11, B=@n12)), Exp=ScalarParameter() {defaultValue=70, parameterName=Spec Highlight Power}) {constExponent=20}
  n14 = ScalarParameter() {defaultValue=3, parameterName=Spec Intense}
  n15 = SetMaterialAttributes!(
   MaterialAttributes
   Base Color=Multiply(A=VectorParameter() {group=Dirt, parameterName=Color, defaultValue=<rgba 0 0 0 1>}.RGB, B=@n4.R)
   Emissive Color=Add(A=Desaturation(None=@n8, Fraction=@n9), B=Multiply(A=@n13, B=@n14))
   Opacity=StaticSwitchParameter(True=Add(A=@n5, B=@n6), False=@n6) {group=Dirt, parameterName=Dirt}
   Normal=Add(A=@n2, B=@n3.RGB)
   Refraction=ScalarParameter() {defaultValue=1, parameterName=Refraction, sliderMax=2}
  ) {attributeSetTypes [
    #1
    #2
    #3
    #4
    #5
   ]}
  MaterialFunctionCall(In=@n15) {materialFunction=&MF_TechArtOverrides_accessories}
 ]
}
