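# Physics asset for the Distant Echo Pro glider mesh, a VinderTech GliderChute cosmetic, giving its skeleton simplified collision and joints.
# One large capsule covers the Frame bone, and two small mirrored capsules cover the left and right handle bases.
# Each handle base has a constraint to the Frame with matching swing and twist limit angles, forming a simple two-branch hierarchy.
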PhysicsAsset /BRCosmetics/Gadgets/Assets/VinderTech_GliderChute/Glider_DistantEchoPro/Meshes/DistantEchoPro_Glider_PhysicsAsset
bodies {
 Frame [{shapeType=Capsule, length=173.4838, radius=162.8697, center=<7.2792 -12.6755 25.7013>, rotation=<rot 0 -140.1779 -87.4885>}]
 Handle_Base_l [{shapeType=Capsule, length=16.6093, radius=13.7248, center=<-5.2146 -1.8288 8.4072>, rotation=<rot 0 109.2339 -158.417>}]
 Handle_Base_r [{shapeType=Capsule, length=16.6093, radius=13.7248, center=<5.2084 1.8287 -8.4072>, rotation=<rot 0 109.2334 -158.417>}]
}
constraints [
 {bone1Name=Handle_Base_r, bone2Name=Frame, swing1LimitDegrees=45, swing2LimitDegrees=45, twistLimitDegrees=45}
 {bone1Name=Handle_Base_l, bone2Name=Frame, swing1LimitDegrees=45, swing2LimitDegrees=45, twistLimitDegrees=45}
]
